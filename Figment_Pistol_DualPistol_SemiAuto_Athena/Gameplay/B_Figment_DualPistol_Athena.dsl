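; Athena dual semi-auto pistol weapon Blueprint (Figment) that customizes firing sound and effects for both guns.
; Custom fire audio spawns the override sound attached to the root, working around burst-style firing.
; Hand fire calls the parent, then after a short delay triggers muzzle flash, shell ejection and force feedback, intended for the other hand.
; The muzzle flash override skips scoped views and activates both muzzles, ignoring its OffHand input.
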(event EventPlayWeaponFireCustomAudio
  (params (bPersistentFire :type Boolean) (bSecondaryFire :type Boolean))
  ;[1 1 1 1 32] Fire sound work around to quickly fix problem with this weapon being "burst" but sound design not. FORT-96639
  (Audio|SpawnSoundAttached (Variables|Default|GetFireSoundOverride) (Variables|Transformation|GetRootComponent)))

(event EventHandFired
  (params (Hand :type EDualWeaponHand%20Enum :default "LEFT") (bPersistantFire :type Boolean))
  (bind _returnvalue (Utilities|Enum|NotEqual(Enum) Hand "LEFT"))
  (|CallParentFunction Hand bPersistantFire)
  ;[0 0 0] Play FX on other Hand, should probably be rolled into base bp/class as an option, if it is already i did not see one
  (Utilities|FlowControl|Delay 0.125)
  (if (not (Variables|Default|GetbStoppedWeaponFireFX))
    (CallFunction|MuzzleFlashFX bPersistantFire _returnvalue)
    (if (and (not (Weapon|IsUsingScopeorFirstPersonCamera)) (> (Rendering|GetRenderingDetailMode) 0))
      (CallFunction|ActivateShellsFX (Variables|Effects%28Shells%29|GetUseShellsOnFire?) _returnvalue)
      (else
        _))
    (FortWeapon|PlayForceFeedback false (Variables|Default|GetbForceFeedbackShouldPersist))))

(fn MuzzleFlashFX
  (params (Persistent%20Fire :type Boolean) (OffHand :type Boolean))
  ;[1 1 1] If local player is using the scope, don't play muzzle effects
  (if (not (Weapon|IsUsingScope))
    (if (not Persistent%20Fire)
      (CallFunction|AddRandomScale false))
    (Components|Activation|Activate (Variables|Effects|GetMuzzle(Empty)) true)
    (CallFunction|SetUpMuzzleFXParams (Variables|Default|GetMuzzle%28Empty%29%20Niagara))
    (Components|Activation|Activate (Variables|Default|GetLeftMuzzle(Empty)) true)
    (CallFunction|SetUpMuzzleFXParams (Variables|Default|GetLeftMuzzle%28Empty%29%20Niagara))
    (Variables|Effects|SetLastPlayFXTime (Utilities|Time|GetGameTimeinSeconds))
    (return)))

(fn ConstructionScript
  (|CallParentFunction))
