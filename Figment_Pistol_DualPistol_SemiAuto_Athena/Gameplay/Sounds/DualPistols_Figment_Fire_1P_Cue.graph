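# First-person firing sound cue for the Figment semi-auto dual pistols.
# A Random node picks among four shared dual-pistol fire waves at equal weight and is excluded from branch culling.
# Playback is scaled below full volume and routed through the first-person ranged-fire sound class, a first-person projectile attenuation asset, and first-person gunfire concurrency.
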SoundCue /Figment_Pistol_DualPistol_SemiAuto_Athena/Gameplay/Sounds/DualPistols_Figment_Fire_1P_Cue
root=Random_2
nodes [
 {
  id=Random_2
  children [WavePlayer_0,WavePlayer_1,WavePlayer_2,WavePlayer_3]
  properties {bShouldExcludeFromBranchCulling=true, weights [1,1,1,1]}
 }
 {id=WavePlayer_0, properties {soundWaveAssetPtr=/Game/Sounds/Fort_Weapons/pistol/Dual/DualPistols_Fire_01}}
 {id=WavePlayer_1, properties {soundWaveAssetPtr=/Game/Sounds/Fort_Weapons/pistol/Dual/DualPistols_Fire_03}}
 {id=WavePlayer_2, properties {soundWaveAssetPtr=/Game/Sounds/Fort_Weapons/pistol/Dual/DualPistols_Fire_02}}
 {id=WavePlayer_3, properties {soundWaveAssetPtr=/Game/Sounds/Fort_Weapons/pistol/Dual/DualPistols_Fire_04}}
]
properties {
 attenuationSettings=&Fort_Weapon_Projectile_FirstPerson_Athena_Att
 maxDistance=2097152
 soundClassObject=&Fort_WeaponRanged_Fire_1P_PSM
 volumeMultiplier=0.65
 attenuationOverrides {distanceAlgorithm=NaturalSound, falloffDistance=5000, attenuationShapeExtents {x=1000}}
 concurrencySet [&Weapon_Gunfire_FirstPerson_Conc]
}
