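# Sound cue for the Figment semi-auto dual pistols' gunfire, built from three distance layers: close, medium and distant.
# Each layer picks a random wave variant, passes through its own attenuation asset and sound class override, then feeds a weighted mixer that favors the close layer.
# Close and medium layers use the third-person fire class, while distant uses a distant fire class.
# Local and global gunfire concurrency sets govern playback.
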SoundCue /Figment_Pistol_DualPistol_SemiAuto_Athena/Gameplay/Sounds/DualPistols_Figment_Fire_Cue
root=Mixer_17
nodes [
 {
  id=Attenuation_0
  children [Random_9]
  properties {attenuationSettings=&Fort_Weapon_Projectile_Medium_Mono_Athena_Att}
 }
 {
  id=Attenuation_34
  children [Random_7]
  properties {attenuationSettings=&Fort_Weapon_Projectile_Distant_Mono_Athena_Att}
 }
 {id=Attenuation_49, children [Random_5], properties {attenuationSettings=&Fort_Weapon_Projectile_Close_Athena_Att}}
 {id=Mixer_17, children [SoundClass_32, SoundClass_0, SoundClass_35], properties {inputVolume [0.7,0.35,0.1]}}
 {
  id=Random_5
  children [WavePlayer_14, WavePlayer_15, WavePlayer_16, WavePlayer_17]
  properties {bShouldExcludeFromBranchCulling=true, weights [1,1,1,1]}
 }
 {id=Random_7, children [WavePlayer_18, WavePlayer_19, WavePlayer_20], properties {bShouldExcludeFromBranchCulling=true, weights [1,1,1]}}
 {id=Random_9, children [WavePlayer_21, WavePlayer_22, WavePlayer_23], properties {bShouldExcludeFromBranchCulling=true, weights [1,1,1]}}
 {
  id=SoundClass_0
  children [Attenuation_0]
  properties {soundClassOverride=&Fort_WeaponRanged_Fire_3P_PSM}
 }
 {
  id=SoundClass_32
  children [Attenuation_49]
  properties {soundClassOverride=&Fort_WeaponRanged_Fire_3P_PSM}
 }
 {
  id=SoundClass_35
  children [Attenuation_34]
  properties {soundClassOverride=&Fort_WeaponRanged_Fire_Distant}
 }
 {id=WavePlayer_14, properties {soundWaveAssetPtr=/Game/Sounds/Fort_Weapons/pistol/Dual/DualPistols_Fire_01}}
 {id=WavePlayer_15, properties {soundWaveAssetPtr=/Game/Sounds/Fort_Weapons/pistol/Dual/DualPistols_Fire_03}}
 {id=WavePlayer_16, properties {soundWaveAssetPtr=/Game/Sounds/Fort_Weapons/pistol/Dual/DualPistols_Fire_02}}
 {id=WavePlayer_17, properties {soundWaveAssetPtr=/Game/Sounds/Fort_Weapons/pistol/Dual/DualPistols_Fire_04}}
 {id=WavePlayer_18, properties {soundWaveAssetPtr=/Game/Sounds/Fort_Weapons/pistol/Dual/DualPistols_Fire_DistantNew_02}}
 {id=WavePlayer_19, properties {soundWaveAssetPtr=/Game/Sounds/Fort_Weapons/pistol/Dual/DualPistols_Fire_DistantNew_01}}
 {id=WavePlayer_20, properties {soundWaveAssetPtr=/Game/Sounds/Fort_Weapons/pistol/Dual/DualPistols_Fire_DistantNew_03}}
 {id=WavePlayer_21, properties {soundWaveAssetPtr=/Game/Sounds/Fort_Weapons/pistol/Dual/DualPistols_Fire_MediumNew_02}}
 {id=WavePlayer_22, properties {soundWaveAssetPtr=/Game/Sounds/Fort_Weapons/pistol/Dual/DualPistols_Fire_MediumNew_01}}
 {id=WavePlayer_23, properties {soundWaveAssetPtr=/Game/Sounds/Fort_Weapons/pistol/Dual/DualPistols_Fire_MediumNew_03}}
]
properties {
 maxDistance=27500
 volumeMultiplier=1
 attenuationOverrides {distanceAlgorithm=NaturalSound, falloffDistance=5000, attenuationShapeExtents {x=1000}}
 concurrencySet [&Weapon_Gunfire_Local_Conc, &Weapon_Gunfire_Global_Conc]
}
