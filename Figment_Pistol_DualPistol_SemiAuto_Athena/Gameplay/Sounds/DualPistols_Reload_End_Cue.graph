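# Sound cue for the end of the dual pistols' reload in the Figment pistol plugin.
# A random node picks between two reload-end wave variations with equal weight.
# It plays as weapon foley with shared reload attenuation, reduced volume, and overridden concurrency that prevents new instances when the limit is reached.
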SoundCue /Figment_Pistol_DualPistol_SemiAuto_Athena/Gameplay/Sounds/DualPistols_Reload_End_Cue
root=Random_0
nodes [
 {id=Random_0, children [WavePlayer_0,WavePlayer_1], properties {weights [1,1]}}
 {id=WavePlayer_0, properties {soundWaveAssetPtr=&/DualPistols_Reload_End_01}}
 {id=WavePlayer_1, properties {soundWaveAssetPtr=&/DualPistols_Reload_End_02}}
]
properties {
 attenuationSettings=&Fort_Weapon_Reload_Att
 bOverrideConcurrency=true
 maxDistance=2250
 soundClassObject=&Fort_WeaponFoley
 volumeMultiplier=0.532
 concurrencyOverrides {resolutionRule=PreventNew}
}
