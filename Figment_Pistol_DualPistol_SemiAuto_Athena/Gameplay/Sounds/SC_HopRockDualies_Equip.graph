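# Sound cue for the equip foley of the Hop Rock Dualies dual pistols.
# A short fixed delay feeds a default Modulator for slight pitch and volume variation, which plays a single equip wave.
# It uses the weapon foley sound class and weapon reload attenuation, with a concurrency override that prevents new instances once the limit is reached.
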SoundCue /Figment_Pistol_DualPistol_SemiAuto_Athena/Gameplay/Sounds/SC_HopRockDualies_Equip
root=Delay_0
nodes [
 {id=Delay_0, children [Modulator_1], properties {delayMax=0.15, delayMin=0.15}}
 {id=Modulator_1, children [WavePlayer_2]}
 {id=WavePlayer_2, properties {soundWaveAssetPtr=/Game/Sounds/Fort_Weapons/pistol/HopRockDualies/WPN_HopRockDualies_Equip_01}}
]
properties {
 attenuationSettings=&Fort_Weapon_Reload_Att
 bOverrideConcurrency=true
 maxDistance=2250
 soundClassObject=&Fort_WeaponFoley
 volumeMultiplier=0.752
 attenuationOverrides {falloffDistance=4100, attenuationShapeExtents {x=900}}
 concurrencyOverrides {resolutionRule=PreventNew}
}
